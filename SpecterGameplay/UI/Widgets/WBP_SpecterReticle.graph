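# Reticle extension widget for SpecterPawn that shows the Specter UpdootItem cooldown as progress on the reticle icon's material.
# Construct caches the dynamic material and owning pawn, then async-listens for the cooldown gameplay tag being added or removed.
# These events start or stop an update timer that feeds remaining and total time to SetCooldownProgress, which writes fill and full-flag material scalars.
# At design time, PreConstruct calls the same function with DesignerCooldownPct.
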WidgetBlueprint /SpecterGameplay/UI/Widgets/WBP_SpecterReticle
graphs [
 EventGraph {nodes [
   n1 = Event PreConstruct()
   > Branch(Condition=@n1.IsDesignTime)
   > SetDesignerPreview()
   Event Construct()
   > n2 = GetDynamicMaterial(self=Get Icon_Reticle() -> (Icon_Reticle:Image Object Reference))
   > Set IconMat(IconMat:Material Instance Dynamic Object Reference=@n2) -> (Output_Get:Material Instance Dynamic Object Reference)
   > n3 = Cast To SpecterPawn!(execute, Object=GetOwningPlayerPawn!(self:User Widget Object Reference) -> (ReturnValue:Pawn Object Reference)) -> (then, CastFailed, AsSpecter Pawn:Specter Pawn Object Reference) {is=title K2Node_DynamicCast}
   > n4 = Set SpecterPawn(SpecterPawn:Specter Pawn Object Reference=@n3) -> (Output_Get:Specter Pawn Object Reference)
   n5 = Get CooldownTag() -> (CooldownTag:Gameplay Tag Structure)
   > n6 = WaitGameplayTagAddToActor!(execute:Exec, TargetActor:Actor Object Reference=@n4, Tag:Gameplay Tag Structure=@n5, OnlyTriggerOnce:bool=false) -> (then:Exec, AsyncTaskProxy:Ability Async Wait Gameplay Tag Added Object Reference, Added:Exec) {is=title K2Node_AsyncAction}
   > n7 = WaitGameplayTagRemoveFromActor!(execute:Exec, TargetActor:Actor Object Reference=@n4, Tag:Gameplay Tag Structure=@n5, OnlyTriggerOnce:bool=false) -> (then:Exec, AsyncTaskProxy:Ability Async Wait Gameplay Tag Removed Object Reference, Removed:Exec) {is=title K2Node_AsyncAction}
   n8 = Get CooldownTimerInterval() -> (CooldownTimerInterval:double)
   HandleCooldownTimer!(execute=@n7.Removed, self:Fort Weapon Reticle Extension Widget Base Object Reference, bActivateTimer:bool=false, TimerInterval:float=@n8) -> (then)
   > UpdateRemainingCooldown(self:Fort Weapon Reticle Extension Widget Base Object Reference)
   HandleCooldownTimer(execute=@n6.Added, bActivateTimer=true, TimerInterval=@n8)
   Event UpdateRemainingCooldown() {is=title K2Node_Event}
   > n9 = GetCooldownTimeRemaining!(execute, self:Fort Weapon Reticle Extension Widget Base Object Reference) -> (then, RemainingTime:float, TotalDuration:float)
   > SetCooldownProgress!(execute, self, RemainingTime:float=@n9, TotalDuration:double=@n9.TotalDuration) -> (then)
  ]}
 SetCooldownProgress {nodes [
   n1 = SetCooldownProgress!() -> (then, RemainingTime:float=~, TotalDuration:double=~) {is=title K2Node_FunctionEntry}
   > n2 = Get!(execute:Exec) -> (then:Exec, else:Exec, IconMat:Material Instance Dynamic Object Reference)
   > SetScalarParameterValue(
    self=@n2
    ParameterName=Get CooldownParameterName() -> (CooldownParameterName:Name=None)
    Value=float - float(A=1, B=SafeDivide(A=@n1, B=@n1.TotalDuration))
   )
   n3 = Get RemainingTime() -> (RemainingTime:float)
   > SetScalarParameterValue(
    self=@n2
    ParameterName=Get ProgressFullParameterName() -> (ProgressFullParameterName:Name=None)
    Value=SelectFloat(A=1, B=0, bPickA="float <= float"(A=@n3) -> (ReturnValue=~))
   )
  ]}
 SetDesignerPreview {nodes [
   SetDesignerPreview() {is=title K2Node_FunctionEntry}
   n1 = Get DesignerCooldownPct() -> (DesignerCooldownPct:double)
   > SetCooldownProgress(RemainingTime=float * float(A=float - float(A=100, B=@n1), B=0.3) {is=title K2Node_PromotableOperator}, TotalDuration=30) {is=title K2Node_CallFunction}
  ]}
]
variables [
 CooldownParameterName {type=string, default=PbarProgress}
 CooldownTag {type=GameplayTag, default {tagName=Specter.UpdootItem.Cooldown}}
 CooldownTimerInterval {type=number, default=0.1}
 DesignerCooldownPct {type=number, default=100}
 IconMat {type=MaterialInstanceDynamic}
 ProgressFullParameterName {type=string, default=b_PbarFull}
 SpecterPawn {type=SpecterPawn, default=null}
]
widgets {
 info {
  inheritedWidgetCount=1
  parentClass=/Script/FortniteUI.FortWeaponReticleExtensionWidgetBase
  rootWidgetClass=&Image
  widgetCount=2
 }
 widgets [
  Icon_Reticle {
   widgetClassPath=&Image
   bIsVariable=true
   widget=Icon_Reticle
   properties {brush {
     bIsDynamicallyLoaded=false
     drawAs=Image
     imageType=FullColor
     mirroring=NoMirror
     resourceName=None
     resourceObject=/SpecterGameplay/UI/Materials/MI_UI_ReticleSpecterSimple
     tiling=NoTile
     imageSize=<64 64>
     margin=<ltrb 0 0 0 0>
     outlineSettings {
      bUseBrushTransparency=false
      roundingType=HalfHeightRadius
      width=0
      color {colorUseRule=UseColor_Specified, specifiedColor=<rgba 0 0 0 0>}
      cornerRadii=<0 0 0 0>
     }
     tintColor {colorUseRule=UseColor_Specified, specifiedColor=<rgba 1 1 1 1>}
     uVRegion {bIsValid=false, max=<0 0>, min=<0 0>}
    }}
  }
  TutorialHighlight {widgetClassPath=&TutorialHighlight}
 ]
}
